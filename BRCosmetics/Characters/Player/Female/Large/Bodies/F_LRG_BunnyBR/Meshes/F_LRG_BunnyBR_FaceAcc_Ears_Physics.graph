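# Physics asset for the hood and ear bones of the large female BunnyBR character's face accessory, likely for secondary motion.
# Capsule bodies chain from head through base and dynamic hood bones, which branch into left and right ear chains.
# The head-to-hood joint swings and twists freely, the hood joint has moderate limits, and ear segments have tight swing and near-zero twist limits.
# The left chain has one more segment than the right.
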PhysicsAsset /BRCosmetics/Characters/Player/Female/Large/Bodies/F_LRG_BunnyBR/Meshes/F_LRG_BunnyBR_FaceAcc_Ears_Physics
bodies {
 base_ear_00_l [{shapeType=Capsule, length=2, radius=2, shapeName=base_ear_00_l_capsule, rotation=<rot 90 0 0>}]
 base_ear_00_r [{shapeType=Capsule, length=2, radius=2, shapeName=base_ear_00_r_capsule, rotation=<rot 90 0 0>}]
 base_hood_up [{shapeType=Capsule, length=0.505, radius=0.505, shapeName=base_hood_up_capsule, rotation=<rot 0 -90 0>}]
 dyn_ear_00_l [{shapeType=Capsule, length=1.2079, radius=1.5869, shapeName=dyn_ear_00_l_capsule, center=<0.1434 0 0>, rotation=<rot 90 0 0>}]
 dyn_ear_00_r [{shapeType=Capsule, length=0.6043, radius=1.5869, shapeName=dyn_ear_00_r_capsule, center=<-0.3566 0 0>, rotation=<rot 90 0 0>}]
 dyn_ear_01_l [{shapeType=Capsule, length=1.4303, radius=1.6796, shapeName=dyn_ear_01_l_capsule, center=<2.5934 0 0>, rotation=<rot 90 0 0>}]
 dyn_ear_01_r [{shapeType=Capsule, length=2, radius=1.7944, shapeName=dyn_ear_01_r_capsule, center=<2.1434 0 0>, rotation=<rot 90 0 0>}]
 dyn_ear_02_l [{shapeType=Capsule, length=1.1684, radius=1.6007, shapeName=dyn_ear_02_l_capsule, center=<0.8728 0 0>, rotation=<rot 90 0 0>}]
 dyn_ear_02_r [{shapeType=Capsule, length=2, radius=1.7944, shapeName=dyn_ear_02_r_capsule, center=<1.6034 0 0>, rotation=<rot 90 0 0>}]
 dyn_ear_03_l [{shapeType=Capsule, length=1.0496, radius=1.691, shapeName=dyn_ear_03_l_capsule, center=<1.22 0 0>, rotation=<rot 90 0 0>}]
 dyn_ear_03_r [{shapeType=Capsule, length=1.6007, radius=1.5869, shapeName=dyn_ear_03_r_capsule, center=<1.8872 0 0>, rotation=<rot 90 0 0>}]
 dyn_ear_04_l [{shapeType=Capsule, length=1.6169, radius=1.5869, shapeName=dyn_ear_04_l_capsule, center=<2.1126 0 0>, rotation=<rot 90 0 0>}]
 dyn_hood_up [{shapeType=Capsule, length=1, radius=1}]
 head [{shapeType=Capsule, length=1, radius=1}]
}
constraints [
 {bone1Name=base_ear_00_l, bone2Name=dyn_hood_up, swing1LimitDegrees=2, swing2LimitDegrees=3, twistLimitDegrees=1}
 {bone1Name=dyn_ear_00_l, bone2Name=base_ear_00_l, swing1LimitDegrees=6, swing2LimitDegrees=3, twistLimitDegrees=0}
 {bone1Name=dyn_ear_01_l, bone2Name=dyn_ear_00_l, swing1LimitDegrees=6, swing2LimitDegrees=3, twistLimitDegrees=0}
 {bone1Name=dyn_ear_02_l, bone2Name=dyn_ear_01_l, swing1LimitDegrees=6, swing2LimitDegrees=3, twistLimitDegrees=0}
 {bone1Name=base_ear_00_r, bone2Name=dyn_hood_up, swing1LimitDegrees=5, swing2LimitDegrees=5, twistLimitDegrees=1}
 {bone1Name=dyn_ear_00_r, bone2Name=base_ear_00_r, swing1LimitDegrees=6, swing2LimitDegrees=3, twistLimitDegrees=0}
 {bone1Name=dyn_ear_01_r, bone2Name=dyn_ear_00_r, swing1LimitDegrees=6, swing2LimitDegrees=3, twistLimitDegrees=0}
 {bone1Name=dyn_ear_02_r, bone2Name=dyn_ear_01_r, swing1LimitDegrees=6, swing2LimitDegrees=3, twistLimitDegrees=0}
 {
  bone1Name=base_hood_up
  bone2Name=head
  swing1LimitDegrees=45
  swing1Motion=Free
  swing2LimitDegrees=45
  swing2Motion=Free
  twistLimitDegrees=45
  twistMotion=Free
 }
 {bone1Name=dyn_hood_up, bone2Name=base_hood_up, swing1LimitDegrees=25, swing2LimitDegrees=25, twistLimitDegrees=45}
 {bone1Name=dyn_ear_03_l, bone2Name=dyn_ear_02_l, swing1LimitDegrees=6, swing2LimitDegrees=3, twistLimitDegrees=0}
 {bone1Name=dyn_ear_04_l, bone2Name=dyn_ear_03_l, swing1LimitDegrees=6, swing2LimitDegrees=3, twistLimitDegrees=0}
 {bone1Name=dyn_ear_03_r, bone2Name=dyn_ear_02_r, swing1LimitDegrees=6, swing2LimitDegrees=3, twistLimitDegrees=0}
]
