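# Physics asset for the face accessory mesh of the large female BunnyBR cosmetic body, covering the bones of its hood.
# Capsule bodies cover the top, back, left and right dynamic hood bones, plus the hood base, head and upper spine.
# Dynamic hood bones have tight angular limits, while head, hood-base and spine joints rotate freely.
# The side bones attach asymmetrically: left to head, right to spine_05, where the back bone also attaches.
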PhysicsAsset /BRCosmetics/Characters/Player/Female/Large/Bodies/F_LRG_BunnyBR/Meshes/F_LRG_BunnyBR_FaceAcc_Physics
bodies {
 base_hood_up [{shapeType=Capsule, length=0.505, radius=0.505, shapeName=base_hood_up_capsule, rotation=<rot 0 -90 0>}]
 dyn_hood_ba [{shapeType=Capsule, length=1.163, radius=1.8689, shapeName=dyn_hood_ba_capsule, center=<1.8 0 -1.14>, rotation=<rot 90 180 -90>}]
 dyn_hood_l [{shapeType=Capsule, length=12.6553, radius=3, shapeName=dyn_hood_l_capsule, center=<5.2973 0 1.2393>, rotation=<rot 90 0 0>}]
 dyn_hood_r [{shapeType=Capsule, length=10.0907, radius=3, shapeName=dyn_hood_r_capsule, center=<4.7426 0 0>, rotation=<rot 90 0 0>}]
 dyn_hood_up [{shapeType=Capsule, length=2, radius=2, shapeName=dyn_hood_up_capsule, center=<10.4048 0 0>, rotation=<rot -90 180 180>}]
 head [{shapeType=Capsule, length=1, radius=1}]
 spine_03 [{shapeType=Capsule, length=25, radius=8.5, shapeName=spine_03_capsule, center=<0 0.5871 0>, rotation=<rot 0 -90 0>}]
 spine_04 [{shapeType=Capsule, length=21.5, radius=8.255, shapeName=spine_04_capsule, center=<3.6984 1.5693 0>, rotation=<rot 0 -90 0>}]
 spine_05 [{
   shapeType=Capsule
   length=14.7028
   radius=8.2193
   shapeName=spine_05_capsule
   center=<4.8689 0.6733 -0.0111>
   rotation=<rot 1.9269 -90 -1.9551>
  }]
}
constraints [
 {bone1Name=dyn_hood_up, bone2Name=base_hood_up, swing1LimitDegrees=1, swing2LimitDegrees=5, twistLimitDegrees=1}
 {
  bone1Name=base_hood_up
  bone2Name=head
  swing1LimitDegrees=45
  swing1Motion=Free
  swing2LimitDegrees=45
  swing2Motion=Free
  twistLimitDegrees=45
  twistMotion=Free
 }
 {bone1Name=dyn_hood_l, bone2Name=head, swing1LimitDegrees=8, swing2LimitDegrees=15, twistLimitDegrees=5}
 {
  bone1Name=head
  bone2Name=spine_05
  swing1LimitDegrees=45
  swing1Motion=Free
  swing2LimitDegrees=45
  swing2Motion=Free
  twistLimitDegrees=45
  twistMotion=Free
 }
 {
  bone1Name=spine_04
  bone2Name=spine_03
  swing1LimitDegrees=45
  swing1Motion=Free
  swing2LimitDegrees=45
  swing2Motion=Free
  twistLimitDegrees=45
  twistMotion=Free
 }
 {
  bone1Name=spine_05
  bone2Name=spine_04
  swing1LimitDegrees=45
  swing1Motion=Free
  swing2LimitDegrees=45
  swing2Motion=Free
  twistLimitDegrees=45
  twistMotion=Free
 }
 {bone1Name=dyn_hood_ba, bone2Name=spine_05, swing1LimitDegrees=5, swing2LimitDegrees=20, twistLimitDegrees=5}
 {bone1Name=dyn_hood_r, bone2Name=spine_05, swing1LimitDegrees=8, swing2LimitDegrees=15, twistLimitDegrees=5}
]
