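# Animation Blueprint for the BunnyBR large female face accessory, adding secondary motion to its hood and ears.
# Copied parent and body poses feed a head-space RigidBody hood simulation, then bone offsets driven by head twist and back-tilt correct it.
# The cached result branches on bRigidBodyEnabled between RigidBody ear physics and Trail controllers weighted by jumping/falling.
# The event graph updates gravity override, skydive, melee and head-versus-spine alphas.
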AnimBlueprint /BRCosmetics/Characters/Player/Female/Large/Bodies/F_LRG_BunnyBR/Meshes/Parts/F_LRG_BunnyBR_FaceAcc_AnimBP
target_skeleton=/Game/Characters/Player/Female/Large/Bodies/F_LRG_BunnyBR/Meshes/Parts/F_LRG_BunnyBR_FaceAcc_Skeleton
graphs [
 AnimGraph {nodes [
   n1 = "Transform (Modify) Bone - Bone: base_hood_r"(
    ComponentPose=Local To Component(LocalPose=Use cached pose 'Hood_Dyn_Cache'())
    Alpha=1
    Translation="0,0,10"
    Rotation="0,0,0"
    Scale="1,1,1"
   ) {
    translationMode=BMM_Additive
    translationSpace=BCS_BoneSpace
    alphaBoolBlend {blendOption=Linear}
    boneToModify {boneName=base_hood_r}
    translation {z=10}
   }
   n2 = RigidBody!(ComponentPose=@n1, Alpha, OverrideWorldGravity:Vector="0,0,-980", ExternalForce) -> (Pose) {
    bClampLinearTranslationLimitToRefPose=true
    bOverrideWorldGravity=true
    overridePhysicsAsset=/BRCosmetics/Characters/Player/Female/Large/Bodies/F_LRG_BunnyBR/Meshes/F_LRG_BunnyBR_FaceAcc_Ears_Physics
    simulationSpace=BaseBoneSpace
    alphaBoolBlend {blendOption=Linear}
    alphaScaleBiasClamp {clampMin=0.25, interpSpeedDecreasing=4, interpSpeedIncreasing=4}
    baseBoneRef {boneName=head}
    componentAppliedLinearAccClamp=<3000 3000 3000>
    componentLinearAccScale=<0.22 0.22 0.5>
    componentLinearVelScale=<0.22 0.22 0.5>
    overrideWorldGravity {z=-980}
    simSpaceSettings {worldAlpha=0.25}
   }
   n3 = "Transform (Modify) Bone - Bone: base_hood_l"(ComponentPose=@n2, Alpha=1, Translation="0,-3,0", Rotation="0,0,0", Scale="1,1,1") {
    translationMode=BMM_Additive
    translationSpace=BCS_BoneSpace
    alphaBoolBlend {blendOption=Linear}
    alphaScaleBiasClamp {bInterpResult=true, bMapRange=true, interpSpeedDecreasing=5, outRange {min=0.75}}
    boneToModify {boneName=base_hood_l}
    translation {y=-3}
   }
   n4 = "Transform (Modify) Bone - Bone: base_hoodOpening_l"(ComponentPose=@n3, Alpha=1, Translation="2,-3,0", Rotation="0,0,0", Scale="1,1,1") {
    translationMode=BMM_Additive
    translationSpace=BCS_BoneSpace
    alphaBoolBlend {blendOption=Linear}
    alphaScaleBiasClamp {bInterpResult=true, bMapRange=true, interpSpeedDecreasing=5, outRange {min=0.75}}
    boneToModify {boneName=base_hoodOpening_l}
    translation=<2 -3>
   }
   n5 = Get Is Jumping Falling() -> (Is Jumping Falling:double)
   n6 = Trail controller - Bone: dyn_ear_04_l!(
    ComponentPose=Local To Component(LocalPose=Use cached pose 'Hood_Dyn_Cache'())
    Alpha=@n5
    RelaxationSpeedScale:float=3
    FakeVelocity:Vector="0,0,0"
   ) -> (Pose) {
    bActorSpaceFakeVel=true
    bShowTrailLocation=true
    bUsePlanarLimit=true
    chainLength=3
    debugLifeTime=0.1
    maxDeltaTime=0.0333
    relaxationSpeedScale=1.2
    alphaScaleBiasClamp {bClampResult=true, bInterpResult=true, clampMax=0.45, clampMin=0.15, interpSpeedDecreasing=50, interpSpeedIncreasing=5}
    baseJoint {boneName=root}
    rotationLimits [{limitMax=<180 180 180>}, {limitMax=<180 180 180>}, {limitMax=<180 180 180>}]
    rotationOffsets [<0 0 0>, <0 0 0>, <0 0 0>]
    trailBone {boneName=dyn_ear_04_l}
    trailRelaxationSpeed {editorCurveData {keys [<at 0 15 RCIM_Linear>,<at 1 6 RCIM_Linear>]}}
   }
   n7 = Trail controller - Bone: dyn_ear_03_r!(ComponentPose=@n6, Alpha=@n5, RelaxationSpeedScale:float=4, FakeVelocity:Vector="0,0,0") -> (Pose) {
    bActorSpaceFakeVel=true
    bShowTrailLocation=true
    bUsePlanarLimit=true
    chainLength=3
    debugLifeTime=0.1
    maxDeltaTime=0.0333
    relaxationSpeedScale=1.2
    alphaScaleBiasClamp {bClampResult=true, bInterpResult=true, clampMax=0.45, clampMin=0.15, interpSpeedDecreasing=50, interpSpeedIncreasing=5}
    baseJoint {boneName=root}
    rotationLimits [{limitMax=<180 180 180>}, {limitMax=<180 180 180>}, {limitMax=<180 180 180>}]
    rotationOffsets [<0 0 0>, <0 0 0>, <0 0 0>]
    trailBone {boneName=dyn_ear_03_r}
    trailRelaxationSpeed {editorCurveData {keys [<at 0 15 RCIM_Linear>,<at 1 6 RCIM_Linear>]}}
   }
   Output Pose(Result=Blend Poses by bool(BlendPose_0=Component To Local(ComponentPose=@n4), BlendPose_1=Component To Local(ComponentPose=@n7), bActiveValue=Get bRigidBodyEnabled()) {blendType=Linear}) {layerGroup=None}
   n8 = Get Skydive() -> (Skydive:bool)
   n9 = RigidBody(
    ComponentPose=Local To Component(LocalPose=Use cached pose 'InputAnimation'())
    Alpha=SelectFloat(A=1, B=1, bPickA=@n8)
    OverrideWorldGravity=SelectVector(A=Get Gravity Override(), B="0,0,-600", bPickA=@n8)
   ) {
    bClampLinearTranslationLimitToRefPose=true
    bOverrideWorldGravity=true
    overridePhysicsAsset=/BRCosmetics/Characters/Player/Female/Large/Bodies/F_LRG_BunnyBR/Meshes/F_LRG_BunnyBR_FaceAcc_Physics
    simulationSpace=BaseBoneSpace
    alphaBoolBlend {blendOption=Linear}
    alphaScaleBiasClamp {bInterpResult=true, clampMin=0.25, interpSpeedDecreasing=4, interpSpeedIncreasing=4}
    baseBoneRef {boneName=head}
    componentAppliedLinearAccClamp=<3000 3000 3000>
    componentLinearAccScale=<0.22 0.22 0.5>
    componentLinearVelScale=<0.22 0.22 0.5>
    overrideWorldGravity {z=-980}
    simSpaceSettings {worldAlpha=0.25}
   }
   n10 = Get head Rot Back() -> (head Rot Back:float)
   n11 = "Transform (Modify) Bone - Bone: dyn_hood_l"(ComponentPose=@n9, Alpha=@n10, Translation="0,0,0", Rotation="0,0,0", Scale="1,1,1") {
    rotationMode=BMM_Additive
    rotationSpace=BCS_BoneSpace
    alphaBoolBlend {blendOption=Linear}
    alphaScaleBiasClamp {bInterpResult=true}
    boneToModify {boneName=dyn_hood_l}
   }
   n12 = "Transform (Modify) Bone - Bone: dyn_hood_r"(ComponentPose=@n11, Alpha=@n10, Translation="0,0,0", Rotation="0,15,0", Scale="1,1,1") {
    rotationMode=BMM_Additive
    rotationSpace=BCS_BoneSpace
    translationMode=BMM_Additive
    translationSpace=BCS_BoneSpace
    alphaBoolBlend {blendOption=Linear}
    alphaScaleBiasClamp {bMapRange=true}
    boneToModify {boneName=dyn_hood_r}
    rotation {yaw=15}
   }
   n13 = "Transform (Modify) Bone - Bone: dyn_hoodOpening_l"(ComponentPose=@n12, Alpha=@n10, Translation="0,5,0", Rotation="0,0,0", Scale="1,1,1") {
    rotationMode=BMM_Additive
    rotationSpace=BCS_BoneSpace
    translationMode=BMM_Additive
    translationSpace=BCS_BoneSpace
    alphaBoolBlend {blendOption=Linear}
    alphaScaleBiasClamp {bInterpResult=true}
    boneToModify {boneName=dyn_hoodOpening_l}
    translation {y=5}
   }
   n14 = "Transform (Modify) Bone - Bone: dyn_hoodOpening_r"(ComponentPose=@n13, Alpha=@n10, Translation="0,-18,0", Rotation="0,0,0", Scale="1,1,1") {
    rotationMode=BMM_Additive
    rotationSpace=BCS_BoneSpace
    translationMode=BMM_Additive
    translationSpace=BCS_BoneSpace
    alphaBoolBlend {blendOption=Linear}
    alphaScaleBiasClamp {bMapRange=true}
    boneToModify {boneName=dyn_hoodOpening_r}
    translation {y=-18}
   }
   n15 = Get head Twist Pos() -> (head Twist Pos:float)
   n16 = "Transform (Modify) Bone - Bone: dyn_hood_r"(ComponentPose=@n14, Alpha=@n15, Translation="0,15,0", Rotation="0,0,0", Scale="1,1,1") {
    translationMode=BMM_Additive
    translationSpace=BCS_BoneSpace
    alphaBoolBlend {blendOption=Linear}
    alphaScaleBiasClamp {
     bInterpResult=true
     bMapRange=true
     interpSpeedDecreasing=5
     interpSpeedIncreasing=5
     inRange {max=0.75, min=0.25}
     outRange {max=0.75, min=0.25}
    }
    boneToModify {boneName=dyn_hood_r}
    translation {y=15}
   }
   n17 = "Transform (Modify) Bone - Bone: dyn_hoodOpening_r"(ComponentPose=@n16, Alpha=@n15, Translation="0,15,0", Rotation="0,0,0", Scale="1,1,1") {
    translationMode=BMM_Additive
    translationSpace=BCS_BoneSpace
    alphaBoolBlend {blendOption=Linear}
    alphaScaleBiasClamp {bInterpResult=true, bMapRange=true, interpSpeedDecreasing=30, interpSpeedIncreasing=0}
    boneToModify {boneName=dyn_hoodOpening_r}
    translation {y=15}
   }
   Save cached pose 'Hood_Dyn_Cache'(Pose=Component To Local(ComponentPose=@n17)) {cacheName=Hood_Dyn_Cache}
   n18 = Get Body Mesh() -> (Body Mesh:Skeletal Mesh Component Object Reference)
   Save cached pose 'InputAnimation'(Pose=Layered blend per bone(BasePose=Copy Pose From Mesh() {bCopyCurves=true, bUseAttachedParent=true}, BlendPoses_0=Copy Pose From Mesh(SourceMeshComponent=@n18)) {layerSetup [{branchFilters [{boneName=upperarm_l}, {boneName=upperarm_r}]}]}) {cacheName=InputAnimation}
   n19 = Get Skydive()
   SelectVector(A=Get GravityOverride(), B="0,0,400", bPickA=@n19)
   SelectFloat(A=0.4, B=0.75, bPickA=@n19)
  ]}
 EventGraph {nodes [
   Event BlueprintUpdateAnimation()
   > n1 = Sequence!(execute) -> (then_0, then_1, then_2:Exec)
   > n2 = Gravity Override State Function DEPRECATED(CharacterPartAnimInstance=Self-Reference(), GameTimeInSeconds=GetGameTimeInSeconds(), GravityOverrideParams=Get Gravity Override Params())
   > Set Gravity Override(Gravity Override=@n2)
   > Set Skydive(Skydive:bool=Get Skydive()) -> (Output_Get:bool)
   > Set Is Jumping Falling(Is Jumping Falling:double=Get Is Jumping Falling()) -> (Output_Get:double)
   > n3 = Branch(Condition=Get bIsPlayingMeleeAnim())
   > Set IsPlayingMelee(IsPlayingMelee:double=1) -> (Output_Get:double)
   Set IsPlayingMelee(execute=@n3.else, IsPlayingMelee=0)
   n4 = Sequence!(execute=@n1.then_1) -> (then_0, then_1, then_2, then_3:Exec)
   > Set bodyMesh(bodyMesh:Skeletal Mesh Component Object Reference=GetPartSkeletalMeshComponent()) -> (Output_Get:Skeletal Mesh Component Object Reference)
   Set head Twist Pos(
    execute=@n4.then_1
    head Twist Pos:float=GetAlphaFromReferenceTransformDelta(
     SkeletalMeshComponent=Get bodyMesh() -> (bodyMesh:Skeletal Mesh Component Object Reference)
     BoneName=head
     BaseName=spine_05
     DeltaComponentType=QuaternionTwist
     InputRange=(X=10,Y=-55)
     TwistAxis="1,0,0"
    )
   ) -> (Output_Get:float)
   Set head Rot Back(
    execute=@n4.then_3
    head Rot Back:float=GetAlphaFromReferenceTransformDelta(SkeletalMeshComponent=Get bodyMesh(), BoneName=head, BaseName=spine_05, InputRange=(X=0,Y=-40), TwistAxis="1,0,0")
   ) -> (Output_Get:float)
   Set head Twist Neg(
    execute=@n4.then_2
    head Twist Neg:float=GetAlphaFromReferenceTransformDelta(
     SkeletalMeshComponent=Get bodyMesh()
     BoneName=head
     BaseName=spine_05
     DeltaComponentType=QuaternionTwist
     InputRange=(X=-10,Y=-55)
     TwistAxis="1,0,0"
    )
   ) -> (Output_Get:float)
  ]}
]
variables [
 Body Mesh {type=SkeletalMeshComponent}
 bodyMesh {type=SkeletalMeshComponent}
 Gravity Override {type=Vector}
 Gravity Override Params {type=GravityOverrideParamsStruct}
 Gravity Override_0 {type=Vector}
 head Rot Back {type=number}
 head Twist Neg {type=number}
 head Twist Pos {type=number}
 Is Jumping Falling {type=number}
 Is Playing Melee Anim {type=boolean}
 IsPlayingMelee {type=number}
 Skydive {type=boolean}
]
